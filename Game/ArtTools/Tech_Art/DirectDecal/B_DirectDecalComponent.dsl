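; Component that projects a direct decal onto a Fortnite building wall by swapping its slot 0 material.
; Update Direct Decal creates a dynamic instance from a layered parent material selected by the wall's texture data count.
; It copies the wall's textures into it, plus decal texture, projection matrix, layer toggles and parameters.
; SaveMaterial caches the owning wall and original material, which Destroy Self restores before removing the component.
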(event Custom|Update%20Direct%20Decal
  (CallFunction|GetParent)
  (CallFunction|CreateDirectDecalMaterial)
  (CallFunction|ApplyMaterialParameters))

(event Custom|Destroy%20Self
  (Rendering|Material|SetMaterial (Variables|/Script/FortniteGame.BuildingSMActor|GetStaticMeshComponent (Variables|Default|GetWallParent)) 0 (Variables|Default|GetOriginalMaterial))
  (Components|DestroyComponent))

(fn ApplyMaterialParameters
  (bind _directdecalmaterial (Variables|Default|GetDirectDecalMaterial))
  (bind _directdecalmaterial_2 (Variables|Default|GetDirectDecalMaterial))
  (bind _outtexturedataarray (Building|GetBuildingTextureData (Variables|Default|GetWallParent) :!class /Script/FortniteGame.BuildingSMActor))
  (for (_array_element _array_index) _outtexturedataarray
    (Rendering|Material|SetTextureParameterValue _directdecalmaterial (Utilities|Array|Get(aref) (Variables|Default|GetTextureParamsDiffuse) _array_index) (Variables|/Script/FortniteGame.BuildingTextureData|GetDiffuse _array_element))
    (Rendering|Material|SetTextureParameterValue _directdecalmaterial (Utilities|Array|Get(aref) (Variables|Default|GetTextureParamsNormals) _array_index) (Variables|/Script/FortniteGame.BuildingTextureData|GetNormal _array_element))
    (Rendering|Material|SetTextureParameterValue _directdecalmaterial (Utilities|Array|Get(aref) (Variables|Default|GetTextureParamsSpeculars) _array_index) (Variables|/Script/FortniteGame.BuildingTextureData|GetSpecular _array_element)))
  (Rendering|Material|SetVectorParameterValue _directdecalmaterial_2 "Matrix_X" (Variables|MaterialParamVars|GetMatrixX) :!class /Script/Engine.MaterialInstanceDynamic)
  (Rendering|Material|SetVectorParameterValue _directdecalmaterial_2 "Matrix_Y" (Variables|MaterialParamVars|GetMatrixY) :!class /Script/Engine.MaterialInstanceDynamic)
  (Rendering|Material|SetVectorParameterValue _directdecalmaterial_2 "Matrix_Z" (Variables|MaterialParamVars|GetMatrixZ) :!class /Script/Engine.MaterialInstanceDynamic)
  (Rendering|Material|SetVectorParameterValue _directdecalmaterial_2 "DD_EnableDirectDecalLayer" (Utilities|Struct|MakeLinearColor (Math|Conversions|ToFloat(Boolean) (Variables|MaterialParamVars|GetbLayer1)) (Math|Conversions|ToFloat(Boolean) (Variables|MaterialParamVars|GetbLayer2)) (Math|Conversions|ToFloat(Boolean) (Variables|MaterialParamVars|GetbLayer3)) (Math|Conversions|ToFloat(Boolean) (Variables|MaterialParamVars|GetbLayer4))) :!class /Script/Engine.MaterialInstanceDynamic)
  (Rendering|Material|SetTextureParameterValue _directdecalmaterial_2 "Diffuse_ProjectedDecal" (Variables|Default|GetDecalDiffuseTexture))
  (Rendering|Material|SetVectorParameterValue _directdecalmaterial_2 "DD_Params" (Variables|MaterialParamVars|GetDD_Params) :!class /Script/Engine.MaterialInstanceDynamic)
  (Rendering|Material|SetVectorParameterValue _directdecalmaterial_2 "DD_Params2" (Variables|MaterialParamVars|GetDD_Params2) :!class /Script/Engine.MaterialInstanceDynamic)
  (Rendering|Material|SetVectorParameterValue _directdecalmaterial_2 "ForwardVector" (Math|Conversions|ToLinearColor(Vector) (Variables|MaterialParamVars|GetForwardVector)) :!class /Script/Engine.MaterialInstanceDynamic))

(fn CreateDirectDecalMaterial
  (bind _outtexturedataarray (Building|GetBuildingTextureData (Variables|Default|GetWallParent) :!class /Script/FortniteGame.BuildingSMActor))
  (bind _returnvalue (Rendering|Material|CreateDynamicMaterialInstance (Utilities|Select :Index (Utilities|Array|Length _outtexturedataarray) :Option%200 "/Game/ArtTools/Tech_Art/DirectDecal/Materials/MI_DirectDecal_1Layer.MI_DirectDecal_1Layer" :Option%201 "/Game/ArtTools/Tech_Art/DirectDecal/Materials/MI_DirectDecal_2Layer.MI_DirectDecal_2Layer" :Option%202 "/Game/ArtTools/Tech_Art/DirectDecal/Materials/MI_DirectDecal_3Layer.MI_DirectDecal_3Layer" :Option%203 "/Game/ArtTools/Tech_Art/DirectDecal/Materials/MI_DirectDecal_4Layer.MI_DirectDecal_4Layer") "None" "None" :!class /Script/Engine.KismetMaterialLibrary))
  (bind _output_get (Variables|Default|SetDirectDecalMaterial _returnvalue))
  (Rendering|Material|SetMaterial (Variables|/Script/FortniteGame.BuildingSMActor|GetStaticMeshComponent (Variables|Default|GetWallParent)) 0 _output_get))

(fn GetParent)

(fn SaveMaterial
  (bind _wallparent (Variables|Default|GetWallParent))
  (bind _staticmeshcomponent (Variables|/Script/FortniteGame.BuildingSMActor|GetStaticMeshComponent _wallparent))
  (bind _returnvalue (Rendering|Material|GetMaterial _staticmeshcomponent 0))
  (bind _asbuilding_wall (Utilities|Casting|CastToBuildingWall (Components|GetOwner)))
  (Variables|Default|SetWallParent _asbuilding_wall)
  (bind _asmaterial_instance_dynamic (Utilities|Casting|CastToMaterialInstanceDynamic _returnvalue)
    (:then
      (Variables|Default|SetOriginalMaterial _asmaterial_instance_dynamic))
    (:CastFailed
      (bind _asmaterial_instance_constant (Utilities|Casting|CastToMaterialInstanceConstant _returnvalue))
      (Variables|Default|SetOriginalMaterial _asmaterial_instance_constant))))
