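; Art-tool placement actor that projects a decal texture onto Fortnite building walls through a B_DirectDecalComponent on each wall.
; Construction script uses a forced wall list, or diffs box-overlapped walls against history to add and remove components.
; It feeds components the box's inverse-matrix axes, negated actor location, texture and mask settings, with lower Priority claiming shared walls.
; Helper functions snap rotation to axes and set scale or aspect-ratio presets.
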(event EventDestroyed
  (Development|PrintString 0 "___________________________________________________Pamagiti"))

(event Custom|Update%20Decals
  (params (Wall :type Actor%20Object%20Reference))
  (bind _box (Variables|Default|GetBox))
  (bind _returnvalue (Transformation|GetWorldTransform _box))
  (bind _returnvalue_2 (Actor|GetComponentbyClass Wall "/Game/ArtTools/Tech_Art/DirectDecal/B_DirectDecalComponent.B_DirectDecalComponent_C"))
  (bind _returnvalue_3 (Math|Transform|ToMatrix(Transform) _returnvalue))
  (bind (_returnvalue_xplane _returnvalue_yplane _returnvalue_zplane) (Utilities|Struct|BreakMatrix (Math|Matrix|GetInverse(Matrix) _returnvalue_3)))
  (bind (_returnvalue_xplane_w _returnvalue_xplane_x _returnvalue_xplane_y _returnvalue_xplane_z) (Utilities|Struct|BreakPlane _returnvalue_xplane))
  (bind (_returnvalue_yplane_w _returnvalue_yplane_x _returnvalue_yplane_y _returnvalue_yplane_z) (Utilities|Struct|BreakPlane _returnvalue_yplane))
  (bind (_returnvalue_zplane_w _returnvalue_zplane_x _returnvalue_zplane_y _returnvalue_zplane_z) (Utilities|Struct|BreakPlane _returnvalue_zplane))
  (bind (_returnvalue_x _returnvalue_y _returnvalue_z) (Math|Vector|BreakVector (Transformation|GetActorLocation)))
  (Utilities|IsValid _returnvalue_2
    (:Is%20Valid
      (bind _output_get (Variables|Default|SetDecalComponent _returnvalue_2))
      (Variables|/Game/ArtTools/Tech_Art/DirectDecal/B_DirectDecalComponent.B_DirectDecalComponent_C|SetMatrixX (Math|Color|MakeColor _returnvalue_xplane_x _returnvalue_xplane_y _returnvalue_xplane_z (* _returnvalue_x -1.0)) _output_get)
      (Variables|/Game/ArtTools/Tech_Art/DirectDecal/B_DirectDecalComponent.B_DirectDecalComponent_C|SetMatrixY (Math|Color|MakeColor _returnvalue_yplane_x _returnvalue_yplane_y _returnvalue_yplane_z (* _returnvalue_y -1.0)) _output_get)
      (Variables|/Game/ArtTools/Tech_Art/DirectDecal/B_DirectDecalComponent.B_DirectDecalComponent_C|SetMatrixZ (Math|Color|MakeColor _returnvalue_zplane_x _returnvalue_zplane_y _returnvalue_zplane_z (* _returnvalue_z -1.0)) _output_get)
      (Variables|/Game/ArtTools/Tech_Art/DirectDecal/B_DirectDecalComponent.B_DirectDecalComponent_C|SetbLayer1 (Variables|Default|MaterialParameters|GetEnable%20Decal%20Layer%201) _output_get)
      (Variables|/Game/ArtTools/Tech_Art/DirectDecal/B_DirectDecalComponent.B_DirectDecalComponent_C|SetbLayer2 (Variables|Default|MaterialParameters|GetEnable%20Decal%20Layer%202) _output_get)
      (Variables|/Game/ArtTools/Tech_Art/DirectDecal/B_DirectDecalComponent.B_DirectDecalComponent_C|SetbLayer3 (Variables|Default|MaterialParameters|GetEnable%20Decal%20Layer%203) _output_get)
      (Variables|/Game/ArtTools/Tech_Art/DirectDecal/B_DirectDecalComponent.B_DirectDecalComponent_C|SetbLayer4 (Variables|Default|MaterialParameters|GetEnable%20Decal%20Layer%204) _output_get)
      (Variables|/Game/ArtTools/Tech_Art/DirectDecal/B_DirectDecalComponent.B_DirectDecalComponent_C|SetDecalDiffuseTexture (Variables|Default|GetDecal%20Diffuse%20Texture) _output_get)
      (Variables|/Game/ArtTools/Tech_Art/DirectDecal/B_DirectDecalComponent.B_DirectDecalComponent_C|SetDD_Params (Math|Color|MakeColor (Variables|Default|MaterialParameters|GetMask%20Shape) (Variables|Default|MaterialParameters|GetMask%20Blend) (Variables|Default|MaterialParameters|GetAngle%20Blend) (Variables|Default|MaterialParameters|GetAngle%20Hardness)) _output_get)
      (Class|BDirectDecalComponent|UpdateDirectDecal _output_get))))

(event Custom|Add%20Decal%20Components
  (params (Wall :type Building%20Wall%20Object%20Reference))
  ;[0 0 0] Add Decal Component
  (Utilities|IsValid (Actor|GetComponentbyClass Wall "/Game/ArtTools/Tech_Art/DirectDecal/B_DirectDecalComponent.B_DirectDecalComponent_C")
    (:Is%20Not%20Valid
      (Game|AddComponentbyClass Wall 0 "/Game/ArtTools/Tech_Art/DirectDecal/B_DirectDecalComponent.B_DirectDecalComponent_C")
      (Variables|/Script/FortniteGame.BuildingSMActor|SetbAllowCustomMaterial true Wall)
      (if (Variables|Default|DebugDecal|GetbOutputLog)
        (Development|PrintString 0 (Utilities|String|Append "Component Added to: " (Utilities|GetDisplayName Wall)))))))

(event Custom|Remove%20Decal%20Components
  (params (Wall :type Building%20Wall%20Object%20Reference))
  (bind _returnvalue (Actor|GetComponentbyClass Wall "/Game/ArtTools/Tech_Art/DirectDecal/B_DirectDecalComponent.B_DirectDecalComponent_C"))
  ;[0 0 0] Remove Decal Component
  (Utilities|IsValid _returnvalue
    (:Is%20Valid
      (Class|BDirectDecalComponent|DestroySelf _returnvalue)
      (Variables|/Script/FortniteGame.BuildingSMActor|SetbAllowCustomMaterial false Wall)
      (if (Variables|Default|DebugDecal|GetbOutputLog)
        (Development|PrintString 0 (Utilities|String|Append "Component Removed from: " (Utilities|GetDisplayName Wall)))))))

(fn %31024
  (bind (_returnvalue_x _returnvalue_y) (Math|Vector|BreakVector (Transformation|GetActorScale3D)))
  (Transformation|SetActorScale3D (Math|Vector|MakeVector 8.0 _returnvalue_y 8.0))
  (Default|UpdateComponentVariables))

(fn 1:1
  (bind (_returnvalue_x _returnvalue_y) (Math|Vector|BreakVector (Transformation|GetActorScale3D)))
  (Transformation|SetActorScale3D (Math|Vector|MakeVector _returnvalue_x _returnvalue_y _returnvalue_x))
  (Default|UpdateComponentVariables))

(fn 1:2
  (bind (_returnvalue_x _returnvalue_y) (Math|Vector|BreakVector (Transformation|GetActorScale3D)))
  (Transformation|SetActorScale3D (Math|Vector|MakeVector (* _returnvalue_x 2.0) _returnvalue_y _returnvalue_x))
  (Default|UpdateComponentVariables))

(fn 2:1
  (bind (_returnvalue_x _returnvalue_y) (Math|Vector|BreakVector (Transformation|GetActorScale3D)))
  (Transformation|SetActorScale3D (Math|Vector|MakeVector _returnvalue_x _returnvalue_y (* _returnvalue_x 2.0)))
  (Default|UpdateComponentVariables))

(fn %3384
  (bind (_returnvalue_x _returnvalue_y) (Math|Vector|BreakVector (Transformation|GetActorScale3D)))
  (Transformation|SetActorScale3D (Math|Vector|MakeVector 3.0 _returnvalue_y 3.0))
  (Default|UpdateComponentVariables))

(fn 4:3
  (bind (_returnvalue_x _returnvalue_y) (Math|Vector|BreakVector (Transformation|GetActorScale3D)))
  (Transformation|SetActorScale3D (Math|Vector|MakeVector _returnvalue_x _returnvalue_y (* _returnvalue_x 0.75)))
  (Default|UpdateComponentVariables))

(fn %3512
  (bind (_returnvalue_x _returnvalue_y) (Math|Vector|BreakVector (Transformation|GetActorScale3D)))
  (Transformation|SetActorScale3D (Math|Vector|MakeVector 4.0 _returnvalue_y 4.0))
  (Default|UpdateComponentVariables))

(fn %3768
  (bind (_returnvalue_x _returnvalue_y) (Math|Vector|BreakVector (Transformation|GetActorScale3D)))
  (Transformation|SetActorScale3D (Math|Vector|MakeVector 6.0 _returnvalue_y 6.0))
  (Default|UpdateComponentVariables))

(fn 8:3
  (bind (_returnvalue_x _returnvalue_y) (Math|Vector|BreakVector (Transformation|GetActorScale3D)))
  (Transformation|SetActorScale3D (Math|Vector|MakeVector (* _returnvalue_x 2.0) _returnvalue_y (* _returnvalue_x 0.75)))
  (Default|UpdateComponentVariables))

(fn AddComponents
  (for _array_element (Variables|Default|GetAddQueue)
    (bind _asbuilding_static_mesh (Utilities|Casting|CastToBuildingSMActor _array_element))
    (Utilities|IsValid (Actor|GetComponentbyClass _asbuilding_static_mesh "/Game/ArtTools/Tech_Art/DirectDecal/B_DirectDecalComponent.B_DirectDecalComponent_C")
      (:Is%20Valid
        (if (or (not (Utilities|IsValid (Variables|/Game/ArtTools/Tech_Art/DirectDecal/B_DirectDecalComponent.B_DirectDecalComponent_C|GetDecalUtilityActor (Actor|GetComponentbyClass _asbuilding_static_mesh "/Game/ArtTools/Tech_Art/DirectDecal/B_DirectDecalComponent.B_DirectDecalComponent_C")) :!class /Script/Engine.KismetSystemLibrary)) (< (Variables|Default|GetPriority) (Variables|Default|GetPriority (Variables|/Game/ArtTools/Tech_Art/DirectDecal/B_DirectDecalComponent.B_DirectDecalComponent_C|GetDecalUtilityActor (Actor|GetComponentbyClass _asbuilding_static_mesh "/Game/ArtTools/Tech_Art/DirectDecal/B_DirectDecalComponent.B_DirectDecalComponent_C")))))
          (Variables|/Game/ArtTools/Tech_Art/DirectDecal/B_DirectDecalComponent.B_DirectDecalComponent_C|SetDecalUtilityActor self (Actor|GetComponentbyClass _asbuilding_static_mesh "/Game/ArtTools/Tech_Art/DirectDecal/B_DirectDecalComponent.B_DirectDecalComponent_C"))
          (Utilities|Array|AddUnique (Variables|Default|GetDecal%20Componenets) (Actor|GetComponentbyClass _asbuilding_static_mesh "/Game/ArtTools/Tech_Art/DirectDecal/B_DirectDecalComponent.B_DirectDecalComponent_C"))))
      (:Is%20Not%20Valid
        (bind _returnvalue (Game|AddComponentbyClass _asbuilding_static_mesh 0 "/Game/ArtTools/Tech_Art/DirectDecal/B_DirectDecalComponent.B_DirectDecalComponent_C"))
        (Class|BDirectDecalComponent|SaveMaterial _returnvalue)
        (Variables|/Game/ArtTools/Tech_Art/DirectDecal/B_DirectDecalComponent.B_DirectDecalComponent_C|SetDecalUtilityActor self _returnvalue)
        (Utilities|Array|AddUnique (Variables|Default|GetDecal%20Componenets) _returnvalue)
        (Variables|/Script/FortniteGame.BuildingSMActor|SetbAllowCustomMaterial true _asbuilding_static_mesh))))
  (return))

(fn CompareNewToHistory
  (Utilities|Array|Clear (Variables|Default|GetDeleteQueue))
  (Variables|Default|SetAddQueue (Variables|Default|GetDecaled%20Actors%20New))
  (for _array_element (Variables|Default|GetDecaled%20Actors%20History)
    (if (Utilities|Array|ContainsItem (Variables|Default|GetDecaled%20Actors%20New) _array_element)
      (Utilities|Array|RemoveItem (Variables|Default|GetAddQueue) _array_element)
      (else
        (Utilities|Array|Add (Variables|Default|GetDeleteQueue) _array_element))))
  (Variables|Default|SetDecaled%20Actors%20History (Variables|Default|GetDecaled%20Actors%20New))
  (return))

(fn DeleteComponents
  (for _array_element (Variables|Default|GetDeleteQueue)
    (bind _asbuilding_static_mesh (Utilities|Casting|CastToBuildingSMActor _array_element))
    (Variables|/Script/FortniteGame.BuildingSMActor|SetbAllowCustomMaterial false _asbuilding_static_mesh)
    (Utilities|IsValid (Actor|GetComponentbyClass _asbuilding_static_mesh "/Game/ArtTools/Tech_Art/DirectDecal/B_DirectDecalComponent.B_DirectDecalComponent_C")
      (:Is%20Valid
        (if (== (Variables|/Game/ArtTools/Tech_Art/DirectDecal/B_DirectDecalComponent.B_DirectDecalComponent_C|GetDecalUtilityActor (Actor|GetComponentbyClass _asbuilding_static_mesh "/Game/ArtTools/Tech_Art/DirectDecal/B_DirectDecalComponent.B_DirectDecalComponent_C")) self)
          (Utilities|Array|RemoveItem (Variables|Default|GetDecal%20Componenets) (Actor|GetComponentbyClass _asbuilding_static_mesh "/Game/ArtTools/Tech_Art/DirectDecal/B_DirectDecalComponent.B_DirectDecalComponent_C"))
          (Class|BDirectDecalComponent|DestroySelf (Actor|GetComponentbyClass _asbuilding_static_mesh "/Game/ArtTools/Tech_Art/DirectDecal/B_DirectDecalComponent.B_DirectDecalComponent_C"))))))
  (return))

(fn GetOverlappingWalls
  ;[0.1406 0 0] Clear Walls Array
  (Utilities|Array|Clear (Variables|Default|GetOverlappedWalls))
  ;[0 0.4385 1] Find Overlapping walls
  (bind _outactors (Collision|ComponentOverlapActors (Variables|Default|GetBox) (Transformation|GetWorldTransform (Variables|Default|GetBox)) (LocalVariables|GetObjectTypes) "/Script/FortniteGame.BuildingWall"))
  (Variables|Default|SetDecaled%20Actors%20New _outactors)
  (return))

(fn SetupGizmos
  (bind (_returnvalue_x _returnvalue_y _returnvalue_z) (Math|Vector|BreakVector (Transformation|GetWorldScale (Variables|Default|GetBox))))
  ;[0 0 0] Set Arrow Size to be consistant
  (Transformation|SetWorldScale3D (Variables|Default|GetArrow) "1.000000,1.000000,1.000000")
  ;[1 1 1] Box Visualization
  (Rendering|SetVisibility (Variables|Default|GetCube) (Variables|Default|DebugDecal|GetbShowPreviewBox) false)
  ;[1 1 1] Plane Visualization
  (Rendering|SetVisibility (Variables|Default|GetPlane) (Variables|Default|DebugDecal|GetbShowPreviewPlane) false)
  (if (Variables|Default|DebugDecal|GetbShowPreviewPlane)
    (Transformation|SetWorldScale3D (Variables|Default|GetPlane) (Math|Vector|MakeVector _returnvalue_x _returnvalue_z _returnvalue_y))
    (bind _returnvalue (Rendering|Material|CreateDynamicMaterialInstance (Variables|Default|GetPlane) 0 "/Game/ArtTools/Tech_Art/DirectDecal/Materials/M_PreviewDecalPlane.M_PreviewDecalPlane" "None" :!class /Script/Engine.PrimitiveComponent))
    (Rendering|Material|SetTextureParameterValue _returnvalue "Diffuse_ProjectedDecal" (Variables|Default|GetDecal%20Diffuse%20Texture))))

(fn UpdateComponentVariables
  (bind _box (Variables|Default|GetBox))
  (bind _returnvalue (Transformation|GetWorldTransform _box))
  (bind _returnvalue_2 (Math|Transform|ToMatrix(Transform) _returnvalue))
  (bind (_returnvalue_xplane _returnvalue_yplane _returnvalue_zplane) (Utilities|Struct|BreakMatrix (Math|Matrix|GetInverse(Matrix) _returnvalue_2)))
  (bind (_returnvalue_xplane_w _returnvalue_xplane_x _returnvalue_xplane_y _returnvalue_xplane_z) (Utilities|Struct|BreakPlane _returnvalue_xplane))
  (bind (_returnvalue_yplane_w _returnvalue_yplane_x _returnvalue_yplane_y _returnvalue_yplane_z) (Utilities|Struct|BreakPlane _returnvalue_yplane))
  (bind (_returnvalue_zplane_w _returnvalue_zplane_x _returnvalue_zplane_y _returnvalue_zplane_z) (Utilities|Struct|BreakPlane _returnvalue_zplane))
  (bind (_returnvalue_x _returnvalue_y _returnvalue_z) (Math|Vector|BreakVector (Transformation|GetActorLocation)))
  (for _array_element (Variables|Default|GetDecal%20Componenets)
    (if (or (== (Variables|/Game/ArtTools/Tech_Art/DirectDecal/B_DirectDecalComponent.B_DirectDecalComponent_C|GetDecalUtilityActor _array_element) self) (< (Variables|Default|GetPriority) (Variables|Default|GetPriority (Variables|/Game/ArtTools/Tech_Art/DirectDecal/B_DirectDecalComponent.B_DirectDecalComponent_C|GetDecalUtilityActor _array_element))))
      (Variables|/Game/ArtTools/Tech_Art/DirectDecal/B_DirectDecalComponent.B_DirectDecalComponent_C|SetDecalUtilityActor self _array_element)
      (Variables|/Game/ArtTools/Tech_Art/DirectDecal/B_DirectDecalComponent.B_DirectDecalComponent_C|SetMatrixX (Math|Color|MakeColor _returnvalue_xplane_x _returnvalue_xplane_y _returnvalue_xplane_z (* _returnvalue_x -1.0)) _array_element)
      (Variables|/Game/ArtTools/Tech_Art/DirectDecal/B_DirectDecalComponent.B_DirectDecalComponent_C|SetMatrixY (Math|Color|MakeColor _returnvalue_yplane_x _returnvalue_yplane_y _returnvalue_yplane_z (* _returnvalue_y -1.0)) _array_element)
      (Variables|/Game/ArtTools/Tech_Art/DirectDecal/B_DirectDecalComponent.B_DirectDecalComponent_C|SetMatrixZ (Math|Color|MakeColor _returnvalue_zplane_x _returnvalue_zplane_y _returnvalue_zplane_z (* _returnvalue_z -1.0)) _array_element)
      (Variables|/Game/ArtTools/Tech_Art/DirectDecal/B_DirectDecalComponent.B_DirectDecalComponent_C|SetbLayer1 (Variables|Default|MaterialParameters|GetEnable%20Decal%20Layer%201) _array_element)
      (Variables|/Game/ArtTools/Tech_Art/DirectDecal/B_DirectDecalComponent.B_DirectDecalComponent_C|SetbLayer2 (Variables|Default|MaterialParameters|GetEnable%20Decal%20Layer%202) _array_element)
      (Variables|/Game/ArtTools/Tech_Art/DirectDecal/B_DirectDecalComponent.B_DirectDecalComponent_C|SetbLayer3 (Variables|Default|MaterialParameters|GetEnable%20Decal%20Layer%203) _array_element)
      (Variables|/Game/ArtTools/Tech_Art/DirectDecal/B_DirectDecalComponent.B_DirectDecalComponent_C|SetbLayer4 (Variables|Default|MaterialParameters|GetEnable%20Decal%20Layer%204) _array_element)
      (Variables|/Game/ArtTools/Tech_Art/DirectDecal/B_DirectDecalComponent.B_DirectDecalComponent_C|SetDecalDiffuseTexture (Variables|Default|GetDecal%20Diffuse%20Texture) _array_element)
      (Variables|/Game/ArtTools/Tech_Art/DirectDecal/B_DirectDecalComponent.B_DirectDecalComponent_C|SetDD_Params (Math|Color|MakeColor (Variables|Default|MaterialParameters|GetMask%20Shape) (Variables|Default|MaterialParameters|GetMask%20Blend) (Variables|Default|MaterialParameters|GetAngle%20Blend) (Variables|Default|MaterialParameters|GetAngle%20Hardness)) _array_element)
      (Variables|/Game/ArtTools/Tech_Art/DirectDecal/B_DirectDecalComponent.B_DirectDecalComponent_C|SetDD_Params2 (Math|Color|MakeColor (Math|Conversions|ToFloat(Boolean) (Variables|Default|DebugDecal|GetDebug%20Mask)) (Variables|Default|MaterialParameters|GetDiffuse%20Alpha%20Multiplier)) _array_element)
      (Variables|/Game/ArtTools/Tech_Art/DirectDecal/B_DirectDecalComponent.B_DirectDecalComponent_C|SetForwardVector (Math|Vector|GetRightVector (Transformation|GetWorldRotation (Variables|Default|GetBox))) _array_element)
      (Class|BDirectDecalComponent|UpdateDirectDecal _array_element)))
  (return))

(fn ConstructionScript
  (CallFunction|SetupGizmos)
  (if (Utilities|Array|IsEmpty (Variables|Default|GetForcedWallsSelection))
    (CallFunction|GetOverlappingWalls)
    (CallFunction|CompareNewtoHistory)
    (CallFunction|DeleteComponents)
    (Default|AddComponents)
    (Default|UpdateComponentVariables)
    (else
      (for _array_element (Variables|Default|GetForcedWallsSelection)
        (CallFunction|AddDecalComponents _array_element)
        (CallFunction|UpdateDecals _array_element)))))

(fn X+
  (Transformation|SetActorRotation "0.000000,-90.000000,0.000000")
  (Default|UpdateComponentVariables))

(fn X-
  (Transformation|SetActorRotation "0.000000,90.000000,0.000000")
  (Default|UpdateComponentVariables))

(fn Y+
  (Transformation|SetActorRotation "0, 0, 0")
  (Default|UpdateComponentVariables))

(fn Y-
  (Transformation|SetActorRotation "0.000000,180.000000,0.000000")
  (Default|UpdateComponentVariables))
