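# Unlit, additive, two-sided material for a preview box in the DirectDecal art tool, with a blue default Color parameter.
# A mask keeps only back faces via an inverted TwoSidedSign and suppresses faces along local Y using the absolute pre-skinned normal, passed through a VertexInterpolator.
# DepthFade multiplies that mask to soften it where the box intersects scene geometry.
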Material /Game/ArtTools/Tech_Art/DirectDecal/Materials/M_PreviewBox
graph {
 properties {blendMode=BLEND_Additive, pixelDepthOffsetMode=PDOM_Legacy, shadingModel=MSM_Unlit, twoSided=true, usedShadingModels=Unlit}
 nodes [
  n1 = Saturate(None=TwoSidedSign())
  n2 = PreSkinnedNormal()
  n3 = DotProduct(A=Abs(None=VertexInterpolator(VS=@n2).PS), B=Constant3Vector() {constant=<rgba 0 1 0 0>})
  Multiply(A=Multiply(A=OneMinus(None=@n1), B=OneMinus(None=@n3)), B=DepthFade())
  VectorParameter() {parameterName=Color, defaultValue=<rgba 0.0861 0.3225 0.7708 0>}
 ]
}
