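# Translucent, unlit, two-sided material that, by its name, previews a decal plane, though no node is wired to a material output.
# The main chain offsets a virtual plane's center along Z by the negated scene-to-pixel depth difference, then derives UVs through VirtualPlaneCoordinates.
# Those UVs sample a checker texture, masked by absolute U only, as both Max inputs read the same component.
# Several unconnected experimental nodes remain.
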Material /Game/ArtTools/Tech_Art/DirectDecal/Materials/M_PreviewDecalPlane
graph {
 properties {blendMode=BLEND_Translucent, pixelDepthOffsetMode=PDOM_Legacy, shadingModel=MSM_Unlit, twoSided=true, usedShadingModels=Unlit}
 nodes [
  BumpOffset(Height=PixelDepth().R)
  ComponentMask(None=TransformPosition() {transformSourceType=TRANSFORMPOSSOURCE_World, transformType=TRANSFORMPOSSOURCE_Camera}) {g=true, r=true}
  Constant4Vector() {constant=<rgba 1 0 0 0>}
  Constant() {r=0.5}
  n1 = Subtract(A=SceneDepth().R, B=PixelDepth().R)
  Frac(None=Divide(A=@n1) {constB=20})
  MaterialFunctionCall() {materialFunction=&CameraDirectionVector}
  n2 = Multiply(A=@n1) {constB=-1}
  n3 = MaterialFunctionCall(
   Plane Normal Axis=Transform(None=Constant3Vector() {constant=<rgba 0 0 1 0>}) {transformSourceType=TRANSFORMSOURCE_Local}
   Plane X Axis=Transform(None=Constant3Vector() {constant=<rgba 1 0 0 0>}) {transformSourceType=TRANSFORMSOURCE_Local}
   Plane Center=TransformPosition(None=MaterialFunctionCall(Z=@n2) {materialFunction=&MakeFloat3}.Result) {transformType=TRANSFORMPOSSOURCE_World}
   UV Scale=ScalarParameter() {defaultValue=100, parameterName=Param_1}
  ) {materialFunction=&VirtualPlaneCoordinates}
  n4 = MaterialFunctionCall(Float2=Abs(None=@n3.UVs)) {materialFunction=&BreakOutFloat2Components}
  n5 = Constant() {r=1}
  Multiply(
   A=TextureSample(UVs=@n3.UVs) {texture=/Game/Packages/Fortress_SharedMaterials/Textures/T_UV_Checker_D}.RGB
   B=If(A=Max(A=@n4.R, B=@n4.R), "A > B"=Constant(), "A == B"=@n5, "A < B"=@n5) {constB=1}
  )
  PixelDepth()
  ScalarParameter() {defaultValue=1.0115}
  ScalarParameter() {defaultValue=-112.3205, parameterName=Param_2}
  SceneDepth()
  TextureObject() {texture=/Game/Characters/Fortress_Character/Textures/T_All_White}
  TextureSampleParameter2D() {parameterName=Diffuse_ProjectedDecal, texture=&T_BlacksRGB}
  WorldPosition()
 ]
}
